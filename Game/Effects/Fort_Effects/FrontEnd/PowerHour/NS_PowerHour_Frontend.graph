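# Niagara system for the Power Hour front-end effect: orange-gold mesh bursts, a pulsing sprite glow and falling sparks.
# Two mesh emitters and the sparks split particles into two perpendicular groups; User.SpawnRate multiplies all spawn amounts and defaults to zero.
# Placement offsets from System.InitialPosOriented shift sideways by a camera aspect-ratio lerp, and System.ScaleTo0 drives particle scale.
# A camera-attach debug emitter remains, disabled.
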NiagaraSystem /Game/Effects/Fort_Effects/FrontEnd/PowerHour/NS_PowerHour_Frontend
dependencies {
 usedDataInterfaces [
  &NiagaraDataInterfaceCurve
  &NiagaraDataInterfaceCamera
  &NiagaraDataInterfaceVector2DCurve
  &NiagaraDataInterfaceColorCurve
 ]
 usedDynamicInputs [&Lerp_Float, &FN_LOD_Distance_Remap]
 usedModules [
  &self:SystemSpawnScript.SystemScriptSource.SystemScriptGraph.NiagaraNodeAssignment_0.SetVariables_#1
  &SystemState
  &self:SystemSpawnScript.SystemScriptSource.SystemScriptGraph.NiagaraNodeAssignment_1.SetVariables_#2
  &self:CameraAspectRatio
 ]
}
emitters {
 AttachToCameraDEBUG {
  bEnabled=false
  emitterUpdateScript [
   &EmitterState()
   &SpawnRate(SpawnRate=5, Use Spawn Probability=false)
   &SpawnBurst_Instantaneous(Spawn Count=1, Use Spawn Probability=false)
  ]
  particleSpawnScript [/Niagara/Modules/Spawn/Initialization/InitializeParticle(
    Lifetime=5
    Module.Write Color=true
    Module.Write Lifetime=true
    Module.Write Mass=true
    Module.Write Material Random=true
    Module.Write Position=false
    Module.Write Scale=false
    Module.Write SpriteRotation=false
    Module.Write SpriteSize=true
    Position=&SimulationPosition()
    Sprite Size=<20 20>
   )]
  particleUpdateScript [
   &ParticleState()
   /SetVariables(Particles.Position=&AddVectorToPosition(
     Position=&ReturnCameraPosition(Coordinate Space:ENiagaraCoordinateSpace=::Simulation, Camera Query:NiagaraDataInterfaceCamera=DataInterface())
     Vector=<0 -50 5>
    )) {isSetParameters=true}
   &ShapeLocation(MaskSpawnGroup=false, Override Seed=false, Shape Origin=$Particles.Position, UseOwnerScale=true) {enabled=false}
   &self:CameraFOV() {enabled=false, module=CameraFOV}
   &self:CameraFOV() {module=CameraFOV001}
   /SetVariables(Particles.CameraDistance:NiagaraFloat=&DistanceBetweenPositions(A=$Particles.Camera Position World, B=$Particles.Position)) {enabled=false, isSetParameters=true}
   &self:ScratchModule() {enabled=false, module=ScratchModule}
   /SetVariables(Particles.Debug:NiagaraFloat=$Particles.CameraDistance) {isSetParameters=true}
   &DynamicMaterialParameters(
    Index 0 Param 1=$Particles.Debug
    Module.Param0WriteEnabled=true
    Module.Param1WriteEnabled=true
    Module.Param2WriteEnabled=true
    Module.Param3WriteEnabled=true
   )
   &Color(Color=&MaskLinearColor(
     Mask Source=1
     Mask Parameter=&MakeCustomIntFromBool(
      Bool=&SetBoolByFloatComparison(A=$Particles.CameraDistance, B=650, Comparison Type=::NewEnumerator2)
      True INT=1
      False INT=0
     )
     Mask=0
     Masked Linear Color=<rgba 1 0 0.8884 1>
     Passthrough Linear Color=<rgba 0 1 0.1861 1>
    )) {enabled=false}
   &Color(Color=&MaskLinearColor(
     Mask Source=1
     Mask Parameter=&MakeCustomIntFromBool(
      Bool=&SetBoolByFloatComparison(
       A=&MakeFloatFromVector(Vector=Linked() {linkedVariable {name=Particles.Camera Position World, type=NiagaraPosition}}, Channel=::NewEnumerator0)
       B=4.4275
       Comparison Type=::NewEnumerator2
      )
      True INT=1
      False INT=0
     )
     Mask=0
     Masked Linear Color=<rgba 1 0 0.8884 1>
     Passthrough Linear Color=<rgba 0 1 0.1861 1>
    )) {enabled=false, module=Color001}
   &ScaleSpriteSize(
    Scale Sprite Size Mode=::NewEnumerator0
    Uniform Scale Factor=&Divide_Float(
     A=/Niagara/DynamicInputs/Camera/DistanceToCamera(Use Owner Position:NiagaraBool=true, Camera Query:NiagaraDataInterfaceCamera=DataInterface())
     B=100
    )
   ) {enabled=false}
  ]
  rendererClasses [&NiagaraSpriteRendererProperties]
  renderers [{rendererClass=&NiagaraSpriteRendererProperties}]
 }
 Big_Down001 {
  emitterUpdateScript [
   &EmitterState()
   &SpawnRate(SpawnRate=&Multiply_Float(A=3, B=$User.SpawnRate), Use Spawn Probability=false)
   &SpawnBurst_Instantaneous(Spawn Count=6, Use Spawn Probability=false, UseLoopCountLimit=false) {enabled=false}
  ]
  particleSpawnScript [
   &PartitionParticles()
   /Niagara/Modules/Spawn/Initialization/V2/InitializeParticle(
    Color=<rgba 1 0.7347 0 1>
    Lifetime Max=0.55
    Lifetime Min=0.3
    Lifetime Mode=::NewEnumerator1
    Mesh Scale=&Multiply_VectorByFloat(
     Vector=&MaskVector(
      Mask Source=1
      Mask Parameter=$Particles.PartitionParticles.Partition
      Mask=0
      Masked Vector=&RandomRangeVector(Minimum=<-0.8 0.5 0.5>, Maximum=<-1.5 1.3 1.3>)
      Passthrough Vector=&RandomRangeVector(Minimum=<0.8 0.5 0.5>, Maximum=<1.5 1.3 1.3>)
     )
     Float=0.4
    )
    Mesh Scale Mode=::NewEnumerator4
    Position Offset=&AddVectorToPosition(Position=$System.InitialPosOriented, Vector=<-130 0 -45>)
    UsePositionOffset=true
    Write Material Random=true
   )
   &InitialMeshOrientation(
    Mesh Axis To Orient=&MaskVector(Mask Source=1, Mask Parameter=$Particles.PartitionParticles.Partition, Mask=0, Masked Vector=<0 0 -1>, Passthrough Vector=<0 -1 0>)
    Mesh Orientation Mode=::NewEnumerator2
    Rotation=&MaskVector(Mask Source=1, Mask Parameter=$Particles.PartitionParticles.Partition, Mask=0, Masked Vector=<0.5 0 0>, Passthrough Vector=<0 0 1>)
    Use Rotation Vector=true
   ) {module=InitialMeshOrientation002}
   /SetVariables(Particles.Random=&RandomRangeFloat(Minimum=0, Maximum=1)) {isSetParameters=true, module=SetVariables_#3}
   &AddVelocity(
    Velocity=&MaskVector(Mask Source=1, Mask Parameter=$Particles.PartitionParticles.Partition, Mask=0, Masked Vector=<1 0 0>, Passthrough Vector=<0 0 1>)
    Velocity Speed Scale=10
   ) {enabled=false}
   /SetVariables(Particles.Position=&AddVectorToPosition(Position=$Particles.Position, Vector=&Lerp_Vector(A=<0 -70 0>, B=<0 -30 0>, Alpha=$System.CameraAspectRatio))) {isSetParameters=true}
  ]
  particleUpdateScript [
   &ParticleState()
   &ShapeLocation(
    Box Size=<0 0 0>
    MaskSpawnGroup=true
    Override Seed=false
    Shape Origin=$Particles.Position
    Shape Primitive=::NewEnumerator2
    Spawn Group:NiagaraInt32=$Particles.PartitionParticles.Partition
    Surface Only=false
    UseOwnerScale=true
   ) {enabled=false}
   &ShapeLocation(
    Box Size=<0 0 25>
    MaskSpawnGroup=true
    Offset=<8 0 7>
    Offset Mode=::NewEnumerator0
    Override Seed=false
    Shape Origin=$Particles.Position
    Shape Primitive=::NewEnumerator2
    Spawn Group:NiagaraInt32=$Particles.PartitionParticles.Partition
    Spawn Group Mask=1
    Surface Only=false
    UseOwnerScale=true
   ) {enabled=false, module=ShapeLocation001}
   &ScaleMeshSize(Scale Factor=&MakeVector(
     X=&FloatFromCurve(
      FloatCurve=DataInterface() {propertyValues {Curve {
         defaultValue=0
         keys [
          {arriveTangent=0.6613, interpMode=RCIM_Cubic, leaveTangent=0.6613, tangentMode=RCTM_User, time=0, value=0.75}
          {arriveTangent=0.0224, interpMode=RCIM_Cubic, leaveTangent=0.0224, tangentMode=RCTM_User, time=1, value=1}
         ]
        }}}
      Scale Curve=$System.ScaleTo0
     )
     Y=&FloatFromCurve(
      FloatCurve=DataInterface() {propertyValues {Curve {keys [{arriveTangent=11.4421, interpMode=RCIM_Cubic, leaveTangent=11.4421, tangentMode=RCTM_User, time=0, value=1}, <at 1 8 RCIM_Cubic>]}}}
      Scale Curve=&FloatFromCurve(
       FloatCurve=DataInterface() {propertyValues {Curve {
          defaultValue=0
          keys [
           {arriveTangent=0.565, interpMode=RCIM_Cubic, leaveTangent=0.565, tangentMode=RCTM_User, time=0, value=0.75}
           {arriveTangent=0.0224, interpMode=RCIM_Cubic, leaveTangent=0.0224, tangentMode=RCTM_User, time=1, value=1}
          ]
         }}}
       Scale Curve=$System.ScaleTo0
      )
     )
     Z=$System.ScaleTo0
    ))
   &ScaleColor(
    Scale Alpha=&Multiply_Float(
     A=10
     B=&FloatFromCurve(FloatCurve=DataInterface() {propertyValues {Curve {keys [<at 0 0>,<at 0.1 0>,<at 0.2 0.8>,<at 1 0>]}}})
    )
    Scale RGB=&Multiply_VectorByFloat(
     Vector=&Multiply_VectorByFloat(Vector=<1 1 1>, Float=&FN_FrontEndFloatModifier(Gameplay=0.5, Front End=0.5))
     Float=&Multiply_Float(
      A=&RandomRangeFloat(Minimum=0, Maximum=1, Evaluation Type=::NewEnumerator1)
      B=&Multiply_Float(A=&FNTimeOfDayScalar(Night Scale=1, Morning Scale=1, Evening Scale=1), B=10)
     )
    )
    ScaleA=true
    ScaleRGB=true
   )
   &DynamicMaterialParameters(
    Index 0 Param 1=&FloatFromCurve(FloatCurve=DataInterface() {propertyValues {Curve {keys [{tangentMode=RCTM_User, time=0, value=1}, {tangentMode=RCTM_User, time=1, value=1}]}}})
    Index 0 Param 2=0.7
    Index 0 Param 3=&FloatFromCurve(FloatCurve=DataInterface() {propertyValues {Curve {keys [
         {arriveTangent=4.0072, interpMode=RCIM_Cubic, leaveTangent=4.0072, tangentMode=RCTM_User, time=0, value=0}
         {arriveTangent=2.7083, interpMode=RCIM_Cubic, leaveTangent=2.7083, time=0.2543, value=0.6225}
         {arriveTangent=1.8473, interpMode=RCIM_Cubic, leaveTangent=1.8473, time=0.6, value=1.625}
         {arriveTangent=1.6993, interpMode=RCIM_Cubic, leaveTangent=1.6993, tangentMode=RCTM_User, time=1, value=2}
        ]}}})
    Index 0 Param 4=5
    Index 1 Param 1=1
    Index 1 Param 2=$Particles.Random
    Index 1 Param 3=1.3
    Index 1 Param 4=1
    Index 2 Param 1=&FloatFromCurve(FloatCurve=DataInterface() {propertyValues {Curve {keys [<at 0 1>,<at 1 0.3>]}}})
    Module.Param0WriteEnabled=true
    Module.Param0WriteEnabled1=true
    Module.Param0WriteEnabled2=true
    Module.Param1WriteEnabled=true
    Module.Param1WriteEnabled1=true
    Module.Param1WriteEnabled2=true
    Module.Param2WriteEnabled=true
    Module.Param2WriteEnabled1=true
    Module.Param2WriteEnabled2=true
    Module.Param3WriteEnabled=true
    Module.Param3WriteEnabled1=true
    Module.Param3WriteEnabled2=true
    Write Parameter Index 1=true
    Write Parameter Index 2=true
   )
   &SolveForcesAndVelocity(Clamp Velocity=false, Limit Acceleration=false)
  ]
  rendererClasses [&NiagaraMeshRendererProperties]
  renderers [{rendererClass=&NiagaraMeshRendererProperties, rendererIndex=0}]
 }
 Glow002 {
  emitterUpdateScript [
   &EmitterState(Life Cycle Mode=::NewEnumerator1, Loop Behavior=::NewEnumerator1, Loop Duration Mode=::NewEnumerator1, UseLoopDelay=false)
   &SpawnRate(SpawnRate=5, Use Spawn Probability=false) {enabled=false}
   &SpawnBurst_Instantaneous(
    Spawn Count=&Multiply_Int(A=1, B=&MakeIntFromFloat(Float=$User.SpawnRate))
    Use Spawn Probability=false
    UseLoopCountLimit=false
   )
  ]
  particleSpawnScript [
   &InitializeParticle(
    Color=<rgba 1 0.5565 0 0.3>
    Position Offset=&AddVectorToPosition(Position=$System.InitialPosOriented, Vector=<-130 0 -55>)
    Sprite Size=<75 75>
    Sprite Size Mode=::NewEnumerator4
    UsePositionOffset=true
    Write Material Random=true
   )
   /SetVariables() {enabled=false, isSetParameters=true}
   &ShapeLocation(
    Box Size=<0 0 400>
    MaskSpawnGroup=false
    Offset=<800 -240 0>
    Offset Mode=::NewEnumerator0
    Override Seed=false
    Shape Primitive=::NewEnumerator2
    Surface Only=false
    UseOwnerScale=true
   ) {enabled=false}
  ]
  particleUpdateScript [
   &ParticleState(Kill Particles When Lifetime Has Elapsed=false, Let Infinitely Lived Particles Die When Emitter Deactivates=true)
   &ScaleSpriteSize(Scale Sprite Size Mode=::NewEnumerator0, Uniform Scale Factor=$System.ScaleTo0)
   /SetVariables(Particles.Position=&AddVectorToPosition(
     Position=&ReturnCameraPosition(Coordinate Space:ENiagaraCoordinateSpace=::Simulation, Camera Query:NiagaraDataInterfaceCamera=DataInterface())
     Vector=<-20 -50 -20>
    )) {enabled=false, isSetParameters=true}
   /Niagara/Noise/Waveform(
    EnableExponent1=false
    Global Phase:NiagaraFloat=&Add_Float(A=$Emitter.Age, B=&RandomRangeFloat(Minimum=0, Maximum=1))
    GlobalClampMax=false
    GlobalClampMin=false
    UseInput1=false
    UsePhase1=false
    "Waveform [1]":ENiagara_Waveforms=::NewEnumerator9
    "[1] Frequency"=7
   )
   &ScaleColor(
    Scale Alpha=&Add_Float(A=0.5, B=&Multiply_Float(A=0.3, B=$Particles.Waveform.Waveform))
    ScaleA=true
    ScaleRGB=true
   )
   &ShapeLocation(
    MaskSpawnGroup=false
    Offset=&Lerp_Vector(A=<0 -70 0>, B=<0 -30 0>, Alpha=$System.CameraAspectRatio)
    Offset Mode=::NewEnumerator0
    Override Seed=false
    Shape Origin=$Particles.Initial.Position
    Sphere Radius=0
    UseOwnerScale=true
   ) {module=ShapeLocation001}
  ]
  rendererClasses [&NiagaraSpriteRendererProperties]
  renderers [
   {rendererClass=&NiagaraSpriteRendererProperties}
   {rendererClass=&NiagaraSpriteRendererProperties, rendererIndex=1}
  ]
 }
 Small_TowardsCenter002 {
  emitterUpdateScript [
   &EmitterState()
   &SpawnRate(SpawnRate=&Multiply_Float(A=5, B=$User.SpawnRate), Use Spawn Probability=false)
   &SpawnBurst_Instantaneous(Spawn Count=6, Use Spawn Probability=false, UseLoopCountLimit=false) {enabled=false}
  ]
  particleSpawnScript [
   &PartitionParticles()
   &InitializeParticle(
    Color=<rgba 1 0.5565 0 1>
    Lifetime Max=0.45
    Lifetime Min=0.28
    Lifetime Mode=::NewEnumerator1
    Mesh Scale=&Multiply_VectorByFloat(
     Vector=&MaskVector(
      Mask Source=1
      Mask Parameter=$Particles.PartitionParticles.Partition
      Mask=0
      Masked Vector=&RandomRangeVector(Minimum=<-0.8 0.5 0.5>, Maximum=<-1.5 1.3 1.3>)
      Passthrough Vector=&RandomRangeVector(Minimum=<0.8 0.5 0.5>, Maximum=<1.5 1.3 1.3>)
     )
     Float=0.2
    )
    Mesh Scale Mode=::NewEnumerator4
    Position Offset=&AddVectorToPosition(Position=$System.InitialPosOriented, Vector=<-130 0 -45>)
    UsePositionOffset=true
    Write Material Random=true
   )
   &InitialMeshOrientation(
    Mesh Axis To Orient=&MaskVector(Mask Source=1, Mask Parameter=$Particles.PartitionParticles.Partition, Mask=0, Masked Vector=<1 0 0>, Passthrough Vector=<0 0 1>)
    Mesh Orientation Mode=::NewEnumerator2
    Rotation=&MaskVector(Mask Source=1, Mask Parameter=$Particles.PartitionParticles.Partition, Mask=0, Masked Vector=<0 0 0>, Passthrough Vector=<0.25 0 1>)
    Use Rotation Vector=true
   ) {enabled=false, module=InitialMeshOrientation002}
   &InitialMeshOrientation(
    Mesh Axis To Orient=&MaskVector(Mask Source=1, Mask Parameter=$Particles.PartitionParticles.Partition, Mask=0, Masked Vector=<0 0 -1>, Passthrough Vector=<0 -1 0>)
    Mesh Orientation Mode=::NewEnumerator2
    Rotation=&MaskVector(Mask Source=1, Mask Parameter=$Particles.PartitionParticles.Partition, Mask=0, Masked Vector=<0.5 0 0>, Passthrough Vector=<0 0 1>)
    Use Rotation Vector=true
   )
   /SetVariables(Particles.Random=&RandomRangeFloat(Minimum=0, Maximum=1)) {isSetParameters=true, module=SetVariables_#3}
   &ShapeLocation(
    Box Size=<0 0 50>
    MaskSpawnGroup=true
    Override Seed=false
    Shape Origin=$Particles.Initial.Position
    Shape Primitive=::NewEnumerator2
    Spawn Group:NiagaraInt32=$Particles.PartitionParticles.Partition
    Surface Only=false
    UseOwnerScale=true
   )
   &ShapeLocation(
    Box Size=<0 100 0>
    MaskSpawnGroup=true
    Override Seed=false
    Shape Origin=$Particles.Initial.Position
    Shape Primitive=::NewEnumerator2
    Spawn Group:NiagaraInt32=$Particles.PartitionParticles.Partition
    Spawn Group Mask=1
    Surface Only=false
    UseOwnerScale=true
   ) {module=ShapeLocation001}
   &AddVelocity(
    Velocity=&MaskVector(Mask Source=1, Mask Parameter=$Particles.PartitionParticles.Partition, Mask=0, Masked Vector=<1 0 0>, Passthrough Vector=<0 0 1>)
    Velocity Speed Scale=6
   ) {enabled=false}
   /SetVariables(Particles.Position=&AddVectorToPosition(Position=$Particles.Position, Vector=&Lerp_Vector(A=<0 -70 0>, B=<0 -30 0>, Alpha=$System.CameraAspectRatio))) {isSetParameters=true}
  ]
  particleUpdateScript [
   &ParticleState()
   /SetVariables(Particles.Position=&AddVectorToPosition(
     Position=&ReturnCameraPosition(Coordinate Space:ENiagaraCoordinateSpace=::Simulation, Camera Query:NiagaraDataInterfaceCamera=DataInterface())
     Vector=<-30 -50 -12>
    )) {enabled=false, isSetParameters=true}
   &ScaleMeshSize(Scale Factor=&MakeVector(
     X=&FloatFromCurve(
      FloatCurve=DataInterface() {propertyValues {Curve {
         defaultValue=0
         keys [
          {arriveTangent=0.6613, interpMode=RCIM_Cubic, leaveTangent=0.6613, tangentMode=RCTM_User, time=0, value=0.75}
          {arriveTangent=0.0224, interpMode=RCIM_Cubic, leaveTangent=0.0224, tangentMode=RCTM_User, time=1, value=1}
         ]
        }}}
      Scale Curve=$System.ScaleTo0
     )
     Y=&FloatFromCurve(
      FloatCurve=DataInterface() {propertyValues {Curve {keys [{arriveTangent=11.4421, interpMode=RCIM_Cubic, leaveTangent=11.4421, tangentMode=RCTM_User, time=0, value=1}, <at 1 8 RCIM_Cubic>]}}}
      Scale Curve=&FloatFromCurve(
       FloatCurve=DataInterface() {propertyValues {Curve {
          defaultValue=0
          keys [
           {arriveTangent=0.565, interpMode=RCIM_Cubic, leaveTangent=0.565, tangentMode=RCTM_User, time=0, value=0.75}
           {arriveTangent=0.0224, interpMode=RCIM_Cubic, leaveTangent=0.0224, tangentMode=RCTM_User, time=1, value=1}
          ]
         }}}
       Scale Curve=$System.ScaleTo0
      )
     )
     Z=$System.ScaleTo0
    ))
   &ScaleColor(
    Scale Alpha=&Multiply_Float(
     A=10
     B=&FloatFromCurve(FloatCurve=DataInterface() {propertyValues {Curve {keys [<at 0 0>,<at 0.1 0>,<at 0.2 0.8>,<at 1 0>]}}})
    )
    Scale RGB=&Multiply_VectorByFloat(
     Vector=&Multiply_VectorByFloat(Vector=<1 1 1>, Float=&FN_FrontEndFloatModifier(Gameplay=0.5, Front End=0.5))
     Float=&Multiply_Float(
      A=&RandomRangeFloat(Minimum=0, Maximum=1, Evaluation Type=::NewEnumerator1)
      B=&Multiply_Float(A=&FNTimeOfDayScalar(Night Scale=1, Morning Scale=1, Evening Scale=1), B=10)
     )
    )
    ScaleA=true
    ScaleRGB=true
   )
   &DynamicMaterialParameters(
    Index 0 Param 1=&FloatFromCurve(FloatCurve=DataInterface() {propertyValues {Curve {keys [{tangentMode=RCTM_User, time=0, value=1}, {tangentMode=RCTM_User, time=1, value=1}]}}})
    Index 0 Param 2=0.7
    Index 0 Param 3=&FloatFromCurve(FloatCurve=DataInterface() {propertyValues {Curve {keys [
         {arriveTangent=4.0072, interpMode=RCIM_Cubic, leaveTangent=4.0072, tangentMode=RCTM_User, time=0, value=0}
         {arriveTangent=3.0012, interpMode=RCIM_Cubic, leaveTangent=3.0012, time=0.2413, value=0.9008}
         {arriveTangent=1.4487, interpMode=RCIM_Cubic, leaveTangent=1.4487, time=0.5923, value=1.7775}
         {arriveTangent=1.6993, interpMode=RCIM_Cubic, leaveTangent=1.6993, tangentMode=RCTM_User, time=1, value=2}
        ]}}})
    Index 0 Param 4=2.5
    Index 1 Param 1=0.5
    Index 1 Param 2=$Particles.Random
    Index 1 Param 3=0.5
    Index 1 Param 4=1
    Index 2 Param 1=&FloatFromCurve(FloatCurve=DataInterface() {propertyValues {Curve {keys [<at 0 1>,<at 1 0.3>]}}})
    Module.Param0WriteEnabled=true
    Module.Param0WriteEnabled1=true
    Module.Param0WriteEnabled2=true
    Module.Param1WriteEnabled=true
    Module.Param1WriteEnabled1=true
    Module.Param1WriteEnabled2=true
    Module.Param2WriteEnabled=true
    Module.Param2WriteEnabled1=true
    Module.Param2WriteEnabled2=true
    Module.Param3WriteEnabled=true
    Module.Param3WriteEnabled1=true
    Module.Param3WriteEnabled2=true
    Write Parameter Index 1=true
    Write Parameter Index 2=true
   )
   &SolveForcesAndVelocity(Clamp Velocity=false, Limit Acceleration=false)
  ]
  rendererClasses [&NiagaraMeshRendererProperties]
  renderers [{rendererClass=&NiagaraMeshRendererProperties, rendererIndex=0}]
 }
 Sparks002 {
  emitterUpdateScript [
   &EmitterState()
   &SpawnRate(SpawnRate=&Multiply_Float(A=20, B=$User.SpawnRate), Use Spawn Probability=false)
  ]
  particleSpawnScript [
   &PartitionParticles()
   &InitializeParticle(
    Color=<rgba 1 0.7347 0 1>
    Lifetime Max=0.4
    Lifetime Min=0.25
    Lifetime Mode=::NewEnumerator1
    Position Offset=&AddVectorToPosition(Position=$System.InitialPosOriented, Vector=<-130 0 -45>)
    Sprite Size Max=<2 2>
    Sprite Size Min=<0.7 0.7>
    Sprite Size Mode=::NewEnumerator2
    UsePositionOffset=true
    Write Material Random=true
   )
   &ShapeLocation(
    Box Size=<0 0 50>
    MaskSpawnGroup=true
    Override Seed=false
    Shape Origin=$Particles.Initial.Position
    Shape Primitive=::NewEnumerator2
    Spawn Group:NiagaraInt32=$Particles.PartitionParticles.Partition
    Surface Only=false
    UseOwnerScale=true
   ) {module=ShapeLocation001}
   &ShapeLocation(
    Box Size=<0 50 0>
    MaskSpawnGroup=true
    Override Seed=false
    Shape Origin=$Particles.Initial.Position
    Shape Primitive=::NewEnumerator2
    Spawn Group:NiagaraInt32=$Particles.PartitionParticles.Partition
    Spawn Group Mask=1
    Surface Only=false
    UseOwnerScale=true
   ) {module=ShapeLocation002}
   /SetVariables(Particles.Position=&AddVectorToPosition(Position=$Particles.Position, Vector=&Lerp_Vector(A=<0 -70 0>, B=<0 -30 0>, Alpha=$System.CameraAspectRatio))) {isSetParameters=true}
   &CurlNoiseForce(Bias Noise=false, Noise Frequency=35, Noise Strength=100, Pan Noise=true, Pan Noise Field=<3 3 0>)
   &ApplyInitialForces()
   &AddVelocity(
    Velocity=&MaskVector(Mask Source=1, Mask Parameter=$Particles.PartitionParticles.Partition, Mask=0, Masked Vector=<0 0 1>, Passthrough Vector=<0 -1 0>)
    Velocity Speed Scale=75
   )
  ]
  particleUpdateScript [
   &ParticleState()
   /SetVariables(Particles.Position=&AddVectorToPosition(
     Position=&ReturnCameraPosition(Coordinate Space:ENiagaraCoordinateSpace=::Simulation, Camera Query:NiagaraDataInterfaceCamera=DataInterface())
     Vector=<-30 -60 -20>
    )) {enabled=false, isSetParameters=true}
   &ScaleSpriteSize(Scale Factor=&Vector2DFromCurve(
     Vector2Curve=DataInterface() {propertyValues {
       XCurve {keys [<at 0 1 RCIM_Cubic>, {arriveTangent=-1.2073, interpMode=RCIM_Cubic, leaveTangent=-1.2073, tangentMode=RCTM_User, time=1, value=0.55}]}
       YCurve {keys [<at 0 1 RCIM_Cubic>, {arriveTangent=-1.2073, interpMode=RCIM_Cubic, leaveTangent=-1.2073, tangentMode=RCTM_User, time=1, value=0.55}]}
      }}
     ScaleCurve=&Multiply_Vector2D_ByFloat(Float=$System.ScaleTo0, Vector2D=<1 1>)
    ))
   &GravityForce(Coordinate Space=::Local, Gravity=<0 0 -600>) {module=GravityForce001}
   &Drag(Use Linear Drag=true, Use Rotational Drag=true)
   &SolveForcesAndVelocity(Clamp Velocity=false, Limit Acceleration=false)
   &ScaleSpriteSizeBySpeed(Max Scale Factor=<1 3>, Velocity Threshold=2500)
   &ScaleColor(
    Scale Mode=1
    Scale RGBA=&Vector4FromCurve(Vector4Curve=DataInterface() {propertyValues {
       AlphaCurve {keys [<at 0 1>,<at 1 0>]}
       BlueCurve {keys [<at 0 0>,<at 0.4172 0>,<at 1 0>]}
       GreenCurve {keys [<at 0 3.6733>,<at 0.4172 1.113>,<at 1 0.0757>]}
       RedCurve {keys [<at 0 5>,<at 0.4172 2>,<at 1 1>]}
      }})
    ScaleRGBA=true
   )
  ]
  rendererClasses [&NiagaraSpriteRendererProperties]
  renderers [{rendererClass=&NiagaraSpriteRendererProperties}]
 }
}
summary {
 systemName=NS_PowerHour_Frontend
 emitters [
  {emitterName=Big_Down001, rendererClasses [&NiagaraMeshRendererProperties]}
  {emitterName=Small_TowardsCenter002, rendererClasses [&NiagaraMeshRendererProperties]}
  {emitterName=Glow002, rendererClasses [&NiagaraSpriteRendererProperties]}
  {emitterName=Sparks002, rendererClasses [&NiagaraSpriteRendererProperties]}
  {bEnabled=false, emitterName=AttachToCameraDEBUG, rendererClasses [&NiagaraSpriteRendererProperties]}
 ]
 userVariables [
  User.bIsFrontEnd {type=NiagaraBool}
  User.bIsFrontEndPreview {type=NiagaraBool}
  User.SpawnRate {type=NiagaraFloat, defaultValue=0, description=""}
 ]
}
